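# Physics asset for the TracePaper medium male torso cosmetic, set up for dynamic motion of its jacket.
# Seven jacket chains around back, sides and front each have a master bone joined to the pelvis by a free constraint, then two capsule segments with swing and tight twist limits.
# Free constraints also link each side's lower segment to its front neighbor.
# Large pelvis, spine and thigh capsules, joined by free constraints, approximate the body.
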PhysicsAsset /BRCosmetics/Characters/Player/Male/Medium/Bodies/M_MED_TracePaper/Meshes/M_MED_TracePaper_Dyn_Torso
bodies {
 dyn_jacket_bk_01 [{shapeType=Capsule, length=13.9536, radius=0.5, shapeName=dyn_jacket_bk_l_02_capsule, center=<7.454 -0.0003 0>, rotation=<rot 0 0 89.9994>}]
 dyn_jacket_bk_02 [{shapeType=Capsule, length=13.9536, radius=0.5, shapeName=dyn_jacket_bk_l_02_capsule, center=<7.454 -0.0003 0>, rotation=<rot 0 0 89.9994>}]
 dyn_jacket_bk_MSTR [{shapeType=Capsule, length=1, radius=0.5, rotation=<rot 90 0 0>}]
 dyn_jacket_bk_l_01 [{shapeType=Capsule, length=13.9536, radius=0.5, shapeName=dyn_jacket_bk_l_01_capsule, center=<7.454 -0.0003 0>, rotation=<rot 0 0 89.9994>}]
 dyn_jacket_bk_l_02 [{shapeType=Capsule, length=13.9536, radius=0.5, shapeName=dyn_jacket_bk_l_02_capsule, center=<7.454 -0.0003 0>, rotation=<rot 0 0 89.9994>}]
 dyn_jacket_bk_l_MSTR [{shapeType=Capsule, length=1, radius=0.5, rotation=<rot 90 0 0>}]
 dyn_jacket_bk_r_01 [
  {shapeType=Capsule, length=13.6631, radius=0.7905, shapeName=dyn_jacket_bk_l_02_capsule, center=<7.454 -0.0003 0>, rotation=<rot 0 0 89.9994>}
  {shapeType=Capsule, length=13.9536, radius=0.5, shapeName=dyn_jacket_bk_l_02_capsule, rotation=<rot 0 0 89.9994>}
 ]
 dyn_jacket_bk_r_02 [{shapeType=Capsule, length=13.9536, radius=0.5, shapeName=dyn_jacket_bk_l_02_capsule, center=<7.454 -0.0003 0>, rotation=<rot 0 0 89.9994>}]
 dyn_jacket_bk_r_MSTR [{shapeType=Capsule, length=1, radius=0.5, rotation=<rot 90 0 0>}]
 dyn_jacket_fr_l_01 [
  {
   shapeType=Capsule
   length=15.2954
   radius=1.2943
   shapeName=dyn_jacket_fr_l_01_capsule
   center=<8.1766 0.0002 0.9808>
   rotation=<rot 0 0 -89.9992>
  }
  {
   shapeType=Capsule
   length=15.2954
   radius=1.2943
   shapeName=dyn_jacket_fr_l_01_capsule
   center=<16.1992 0 0.9807>
   rotation=<rot 0 0 -89.9992>
  }
 ]
 dyn_jacket_fr_l_02 [{
   shapeType=Capsule
   length=9.863
   radius=1.5784
   shapeName=dyn_jacket_fr_l_02_capsule
   center=<10.1831 -0.0005 0.6053>
   rotation=<rot 0 0 89.9992>
  }]
 dyn_jacket_fr_l_MSTR [{shapeType=Capsule, length=1, radius=0.5, rotation=<rot 90 0 0>}]
 dyn_jacket_fr_r_01 [
  {
   shapeType=Capsule
   length=14.8003
   radius=1.747
   shapeName=dyn_jacket_fr_r_01_capsule
   center=<15.8584 0 2.0952>
   rotation=<rot 0 0 -89.9993>
  }
  {
   shapeType=Capsule
   length=16.6449
   radius=1.0261
   shapeName=dyn_jacket_fr_r_01_capsule
   center=<8.4581 0 -0.0002>
   rotation=<rot 0 0 -89.9998>
  }
 ]
 dyn_jacket_fr_r_02 [
  {
   shapeType=Capsule
   length=9.4093
   radius=2.0321
   shapeName=dyn_jacket_fr_l_02_capsule
   center=<10.325 -0.0006 2.5807>
   rotation=<rot 0 0 89.9992>
  }
  {
   shapeType=Capsule
   length=5.9428
   radius=0.5
   shapeName=dyn_jacket_fr_r_02_capsule
   center=<7.6513 7.5639 -1.1826>
   rotation=<rot 0 0 -49.9997>
  }
 ]
 dyn_jacket_fr_r_MSTR [{shapeType=Capsule, length=1, radius=0.5, rotation=<rot 90 0 0>}]
 dyn_jacket_l_01 [
  {shapeType=Capsule, length=13.9812, radius=0.5, shapeName=dyn_jacket_l_01_capsule, center=<7.2757 0 0.1868>, rotation=<rot 0 0 89.9979>}
  {shapeType=Capsule, length=13.9812, radius=0.5, shapeName=dyn_jacket_l_01_capsule, rotation=<rot 0 0 89.9979>}
 ]
 dyn_jacket_l_02 [{shapeType=Capsule, length=13.9812, radius=0.5, shapeName=dyn_jacket_l_02_capsule, center=<7.2757 0 0.1868>, rotation=<rot 0 0 89.9979>}]
 dyn_jacket_l_MSTR [{shapeType=Capsule, length=1, radius=0.5, rotation=<rot 90 0 0>}]
 dyn_jacket_r_01 [{shapeType=Capsule, length=10.1567, radius=1.3883, shapeName=dyn_jacket_l_02_capsule, center=<4.4992 0 -0.0704>, rotation=<rot 0 0 89.9979>}]
 dyn_jacket_r_02 [
  {
   shapeType=Capsule
   length=4.7281
   radius=0.5
   shapeName=dyn_jacket_r_02_capsule
   center=<0.8906 -6.0115 -1.0013>
   rotation=<rot 0 0 -119.9978>
  }
  {shapeType=Capsule, length=12.1358, radius=0.7616, shapeName=dyn_jacket_r_02_capsule, center=<6.6755 0 0>, rotation=<rot 0 0 -89.9993>}
 ]
 dyn_jacket_r_MSTR [{shapeType=Capsule, length=1, radius=0.5, rotation=<rot 90 0 0>}]
 pelvis [
  {shapeType=Capsule, length=6.7188, radius=10.9688, center=<-2.0733 -1.3709 0>}
  {shapeType=Capsule, length=9, radius=8.5, shapeName=pelvis_capsule, center=<7.2135 0 -0.0001>}
 ]
 spine_01 [{shapeType=Capsule, length=6.7188, radius=10.9688, center=<-2.0733 -1.3709 0>}]
 spine_02 [{shapeType=Capsule, length=19.1562, radius=12.0938, center=<0 -0.6181 0>}]
 spine_03 [{shapeType=Capsule, length=19.1562, radius=12.0938, center=<0 -0.2866 0>}]
 spine_04 [{shapeType=Capsule, length=19.1562, radius=12.0938, center=<0 -0.1776 0>}]
 spine_05 [{shapeType=Capsule, length=20, radius=13.2188, center=<-4.5203 -1.893 0>}]
 thigh_l [{shapeType=Capsule, length=26.75, radius=8.625, center=<-13.9111 0.0005 0.9087>, rotation=<rot 90 0 0>}]
 thigh_r [{shapeType=Capsule, length=21.4869, radius=9.6068, center=<12.8169 0.2182 -0.9089>, rotation=<rot 90 0 0>}]
}
constraints [
 {bone1Name=dyn_jacket_l_01, bone2Name=dyn_jacket_l_MSTR, swing1LimitDegrees=50, swing2LimitDegrees=35, twistLimitDegrees=5}
 {bone1Name=dyn_jacket_l_02, bone2Name=dyn_jacket_l_01, swing1LimitDegrees=60, swing2LimitDegrees=60, twistLimitDegrees=5}
 {bone1Name=dyn_jacket_fr_l_01, bone2Name=dyn_jacket_fr_l_MSTR, swing1LimitDegrees=5, swing2LimitDegrees=45, twistLimitDegrees=5}
 {bone1Name=dyn_jacket_fr_l_02, bone2Name=dyn_jacket_fr_l_01, swing1LimitDegrees=5, swing2LimitDegrees=30, twistLimitDegrees=5}
 {bone1Name=dyn_jacket_bk_l_01, bone2Name=dyn_jacket_bk_l_MSTR, swing1LimitDegrees=50, swing2LimitDegrees=35, twistLimitDegrees=5}
 {bone1Name=dyn_jacket_bk_l_02, bone2Name=dyn_jacket_bk_l_01, swing1LimitDegrees=60, swing2LimitDegrees=60, twistLimitDegrees=5}
 {bone1Name=dyn_jacket_bk_01, bone2Name=dyn_jacket_bk_MSTR, swing1LimitDegrees=50, swing2LimitDegrees=35, twistLimitDegrees=5}
 {bone1Name=dyn_jacket_bk_02, bone2Name=dyn_jacket_bk_01, swing1LimitDegrees=60, swing2LimitDegrees=60, twistLimitDegrees=5}
 {bone1Name=dyn_jacket_fr_r_01, bone2Name=dyn_jacket_fr_r_MSTR, swing1LimitDegrees=30, swing2LimitDegrees=45, twistLimitDegrees=5}
 {bone1Name=dyn_jacket_fr_r_02, bone2Name=dyn_jacket_fr_r_01, swing1LimitDegrees=30, swing2LimitDegrees=60, twistLimitDegrees=5}
 {bone1Name=dyn_jacket_bk_r_01, bone2Name=dyn_jacket_bk_r_MSTR, swing1LimitDegrees=50, swing2LimitDegrees=35, twistLimitDegrees=5}
 {bone1Name=dyn_jacket_bk_r_02, bone2Name=dyn_jacket_bk_r_01, swing1LimitDegrees=60, swing2LimitDegrees=60, twistLimitDegrees=5}
 {bone1Name=dyn_jacket_r_01, bone2Name=dyn_jacket_r_MSTR, swing1LimitDegrees=50, swing2LimitDegrees=35, twistLimitDegrees=5}
 {bone1Name=dyn_jacket_r_02, bone2Name=dyn_jacket_r_01, swing1LimitDegrees=60, swing2LimitDegrees=60, twistLimitDegrees=5}
 {
  bone1Name=spine_05
  bone2Name=spine_04
  swing1LimitDegrees=45
  swing1Motion=Free
  swing2LimitDegrees=45
  swing2Motion=Free
  twistLimitDegrees=45
  twistMotion=Free
 }
 {
  bone1Name=spine_04
  bone2Name=spine_03
  swing1LimitDegrees=45
  swing1Motion=Free
  swing2LimitDegrees=45
  swing2Motion=Free
  twistLimitDegrees=45
  twistMotion=Free
 }
 {
  bone1Name=dyn_jacket_l_MSTR
  bone2Name=pelvis
  swing1LimitDegrees=45
  swing1Motion=Free
  swing2LimitDegrees=45
  swing2Motion=Free
  twistLimitDegrees=45
  twistMotion=Free
 }
 {
  bone1Name=dyn_jacket_fr_l_MSTR
  bone2Name=pelvis
  swing1LimitDegrees=45
  swing1Motion=Free
  swing2LimitDegrees=45
  swing2Motion=Free
  twistLimitDegrees=45
  twistMotion=Free
 }
 {
  bone1Name=dyn_jacket_bk_l_MSTR
  bone2Name=pelvis
  swing1LimitDegrees=45
  swing1Motion=Free
  swing2LimitDegrees=45
  swing2Motion=Free
  twistLimitDegrees=45
  twistMotion=Free
 }
 {
  bone1Name=dyn_jacket_bk_MSTR
  bone2Name=pelvis
  swing1LimitDegrees=45
  swing1Motion=Free
  swing2LimitDegrees=45
  swing2Motion=Free
  twistLimitDegrees=45
  twistMotion=Free
 }
 {
  bone1Name=dyn_jacket_fr_r_MSTR
  bone2Name=pelvis
  swing1LimitDegrees=45
  swing1Motion=Free
  swing2LimitDegrees=45
  swing2Motion=Free
  twistLimitDegrees=45
  twistMotion=Free
 }
 {
  bone1Name=dyn_jacket_bk_r_MSTR
  bone2Name=pelvis
  swing1LimitDegrees=45
  swing1Motion=Free
  swing2LimitDegrees=45
  swing2Motion=Free
  twistLimitDegrees=45
  twistMotion=Free
 }
 {
  bone1Name=dyn_jacket_r_MSTR
  bone2Name=pelvis
  swing1LimitDegrees=45
  swing1Motion=Free
  swing2LimitDegrees=45
  swing2Motion=Free
  twistLimitDegrees=45
  twistMotion=Free
 }
 {
  bone1Name=thigh_l
  bone2Name=pelvis
  swing1LimitDegrees=45
  swing1Motion=Free
  swing2LimitDegrees=45
  swing2Motion=Free
  twistLimitDegrees=45
  twistMotion=Free
 }
 {
  bone1Name=thigh_r
  bone2Name=pelvis
  swing1LimitDegrees=45
  swing1Motion=Free
  swing2LimitDegrees=45
  swing2Motion=Free
  twistLimitDegrees=45
  twistMotion=Free
 }
 {
  bone1Name=spine_03
  bone2Name=spine_02
  swing1LimitDegrees=45
  swing1Motion=Free
  swing2LimitDegrees=45
  swing2Motion=Free
  twistLimitDegrees=45
  twistMotion=Free
 }
 {
  bone1Name=spine_02
  bone2Name=spine_01
  swing1LimitDegrees=45
  swing1Motion=Free
  swing2LimitDegrees=45
  swing2Motion=Free
  twistLimitDegrees=45
  twistMotion=Free
 }
 {
  bone1Name=spine_01
  bone2Name=pelvis
  swing1LimitDegrees=45
  swing1Motion=Free
  swing2LimitDegrees=45
  swing2Motion=Free
  twistLimitDegrees=45
  twistMotion=Free
 }
 {
  bone1Name=dyn_jacket_fr_r_02
  bone2Name=dyn_jacket_r_02
  swing1LimitDegrees=45
  swing1Motion=Free
  swing2LimitDegrees=45
  swing2Motion=Free
  twistLimitDegrees=45
  twistMotion=Free
 }
 {
  bone1Name=dyn_jacket_l_02
  bone2Name=dyn_jacket_fr_l_02
  swing1LimitDegrees=45
  swing1Motion=Free
  swing2LimitDegrees=45
  swing2Motion=Free
  twistLimitDegrees=45
  twistMotion=Free
 }
]
