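# Post-process style AnimBP for the TracePaper male body cosmetic, animating jacket, tie and sleeve over the pose copied from the parent mesh.
# Pose-driven bone corrections and a Control Rig feed a cached pre-sim pose.
# That pose drives chained torso and tie RigidBody sims with skydiving forces and offsets, or tie-to-spine constraints when dynamics are disabled by LOD or setting.
# Back-jacket pitch offsets finish the chain; EventGraph computes state and thigh/wrist alphas.
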AnimBlueprint /BRCosmetics/Characters/Player/Male/Medium/Bodies/M_MED_TracePaper/Meshes/M_MED_TracePaper_AnimBP
target_skeleton=&/M_MED_TracePaper_Skeleton
graphs [
 AnimGraph {nodes [
   n1 = Break Fort Anim Input Skydiving External Force(FortAnimInput_SkydivingExternalForce=Get SkydivingExternalForceInput()) -> (
    bUseSkydivingVectorForce=~
    SkydivingVectorForce=~
    HeadToPelvisDirection=~
    FloatingMultiplier=~
    FloatingAdditive=~
    DivingMultiplier=~
    DivingAdditive=~
    ParachutingMultiplier=~
    ParachutingAdditive=~
   )
   n2 = RigidBody(ComponentPose=Local To Component(LocalPose=Use cached pose 'pre_sim_cache'()), ExternalForce=@n1.SkydivingVectorForce) {
    bClampLinearTranslationLimitToRefPose=true
    bOverrideWorldGravity=true
    overridePhysicsAsset=&/M_MED_TracePaper_Dyn_Torso
    simulationSpace=BaseBoneSpace
    alphaBoolBlend {blendOption=Linear}
    baseBoneRef {boneName=pelvis}
    componentAppliedLinearAccClamp=<3000 3000 3000>
    componentLinearAccScale=<0.4 0.3 0.2>
    componentLinearVelScale=<0.4 0.3 0.2>
    overrideWorldGravity {z=-980}
   }
   n3 = RigidBody(ComponentPose=@n2, ExternalForce=@n1.SkydivingVectorForce) {
    bClampLinearTranslationLimitToRefPose=true
    bOverrideWorldGravity=true
    overridePhysicsAsset=&/M_MED_TracePaper_Dyn_Tie
    simulationSpace=BaseBoneSpace
    alphaBoolBlend {blendOption=Linear}
    baseBoneRef {boneName=spine_03}
    componentAppliedLinearAccClamp=<3000 3000 3000>
    componentLinearAccScale=<0.4 0.3 0.2>
    componentLinearVelScale=<0.4 0.3 0.2>
    overrideWorldGravity {z=-980}
   }
   n4 = "Transform (Modify) Bone - Bone: dyn_jacket_bk_01"(ComponentPose=@n3, Alpha=Get skydiving alpha() -> (skydiving alpha:double), Translation="0,0,0", Rotation="0,0.000,12", Scale="1,1,1") {rotationMode=BMM_Additive, alphaBoolBlend {blendOption=Linear}, boneToModify {boneName=dyn_jacket_bk_01}, rotation {roll=12}}
   n5 = "Transform (Modify) Bone - Bone: dyn_jacket_bk_r_01"(ComponentPose=@n4, Alpha=Get skydiving alpha(), Translation="0,0,0", Rotation="0,0,12", Scale="1,1,1") {rotationMode=BMM_Additive, alphaBoolBlend {blendOption=Linear}, boneToModify {boneName=dyn_jacket_bk_r_01}, rotation {roll=12}}
   n6 = "Transform (Modify) Bone - Bone: dyn_jacket_bk_l_01"(ComponentPose=@n5, Alpha=Get skydiving alpha(), Translation="0,0,0", Rotation="0,0,12", Scale="1,1,1") {rotationMode=BMM_Additive, alphaBoolBlend {blendOption=Linear}, boneToModify {boneName=dyn_jacket_bk_l_01}, rotation {roll=12}}
   n7 = "Transform (Modify) Bone - Bone: dyn_jacket_r_MSTR"(ComponentPose=@n6, Alpha=Get skydiving alpha(), Translation="0,0,0", Rotation="1,0.000,0.000", Scale="1,1,1") {rotationMode=BMM_Additive, alphaBoolBlend {blendOption=Linear}, boneToModify {boneName=dyn_jacket_r_MSTR}, rotation {pitch=1}}
   n8 = "Transform (Modify) Bone - Bone: dyn_jacket_l_MSTR"(ComponentPose=@n7, Alpha=Get skydiving alpha(), Translation="0,0,0", Rotation="1,0.000,0", Scale="1,1,1") {rotationMode=BMM_Additive, alphaBoolBlend {blendOption=Linear}, boneToModify {boneName=dyn_jacket_l_MSTR}, rotation {pitch=1}}
   n9 = "Transform (Modify) Bone - Bone: dyn_jacket_fr_r_MSTR"(ComponentPose=@n8, Alpha=Get RightThighUpAlpha() -> (RightThighUpAlpha:float), Translation="0,0,0", Rotation="5,0,0", Scale="1,1,1") {
    rotationMode=BMM_Additive
    rotationSpace=BCS_BoneSpace
    translationMode=BMM_Additive
    translationSpace=BCS_BoneSpace
    alphaBoolBlend {blendOption=Linear}
    boneToModify {boneName=dyn_jacket_fr_r_MSTR}
    rotation {pitch=5}
   }
   n10 = "Transform (Modify) Bone - Bone: dyn_jacket_r_MSTR"(ComponentPose=@n9, Alpha=Get RightThighUpAlpha(), Translation="0,0,0", Rotation="5,0,0", Scale="1,1,1") {
    rotationMode=BMM_Additive
    rotationSpace=BCS_BoneSpace
    translationMode=BMM_Additive
    translationSpace=BCS_BoneSpace
    alphaBoolBlend {blendOption=Linear}
    boneToModify {boneName=dyn_jacket_r_MSTR}
    rotation {pitch=5}
   }
   n11 = Component To Local(ComponentPose=@n10)
   n12 = "Constraint - dyn_tie_01 (P)"(ComponentPose=Local To Component(LocalPose=Use cached pose 'pre_sim_cache'()), Alpha=1, ConstraintWeights_0=1) {
    alphaBoolBlend {blendOption=Linear}
    boneToModify {boneName=dyn_tie_01}
    constraintSetup [{targetBone {boneName=spine_05}}]
    constraintWeights [1]
   }
   n13 = "Constraint - dyn_tie_02 (P)"(ComponentPose=@n12, Alpha=1, ConstraintWeights_0=1) {
    alphaBoolBlend {blendOption=Linear}
    boneToModify {boneName=dyn_tie_02}
    constraintSetup [{targetBone {boneName=spine_04}}]
    constraintWeights [1]
   }
   n14 = "Constraint - dyn_tie_03 (P)"(ComponentPose=@n13, Alpha=1, ConstraintWeights_0=1) {
    alphaBoolBlend {blendOption=Linear}
    boneToModify {boneName=dyn_tie_03}
    constraintSetup [{targetBone {boneName=spine_03}}]
    constraintWeights [1]
   }
   n15 = "Constraint - dyn_tie_04 (P)"(ComponentPose=@n14, Alpha=1, ConstraintWeights_0=1) {
    alphaBoolBlend {blendOption=Linear}
    boneToModify {boneName=dyn_tie_04}
    constraintSetup [{targetBone {boneName=spine_02}}]
    constraintWeights [1]
   }
   n16 = "Constraint - dyn_tie_05 (P)"(ComponentPose=@n15, Alpha=1, ConstraintWeights_0=1) {
    alphaBoolBlend {blendOption=Linear}
    boneToModify {boneName=dyn_tie_05}
    constraintSetup [{targetBone {boneName=spine_01}}]
    constraintWeights [1]
   }
   n17 = Component To Local(ComponentPose=@n16)
   n18 = Get dynamics_disabled() -> (dynamics_disabled:bool)
   n19 = "Transform (Modify) Bone - Bone: dyn_jacket_bk_02"!(
    ComponentPose=Local To Component(LocalPose=Blend Poses by bool(BlendPose_0=@n17, BlendPose_1=@n11, bActiveValue=@n18) {blendType=Linear})
    bAlphaBoolEnabled:bool=True
    Translation="0,0,0"
    Rotation="-30,0,0"
    Scale="1,1,1"
   ) -> (Pose) {
    alphaInputType=Bool
    rotationMode=BMM_Additive
    rotationSpace=BCS_BoneSpace
    alphaBoolBlend {blendInTime=0.3, blendOption=CircularIn, blendOutTime=0.3}
    boneToModify {boneName=dyn_jacket_bk_02}
    rotation {pitch=-30}
   }
   n20 = "Transform (Modify) Bone - Bone: dyn_jacket_bk_r_02"!(ComponentPose=@n19, bAlphaBoolEnabled:bool=True, Translation="0,0,0", Rotation="-30,0,0", Scale="1,1,1") -> (Pose) {
    alphaInputType=Bool
    rotationMode=BMM_Additive
    rotationSpace=BCS_BoneSpace
    alphaBoolBlend {blendInTime=0.3, blendOption=CircularIn, blendOutTime=0.3}
    boneToModify {boneName=dyn_jacket_bk_r_02}
    rotation {pitch=-30}
   }
   n21 = "Transform (Modify) Bone - Bone: dyn_jacket_bk_l_02"!(ComponentPose=@n20, bAlphaBoolEnabled:bool=True, Translation="0,0,0", Rotation="-30,0,0", Scale="1,1,1") -> (Pose) {
    alphaInputType=Bool
    rotationMode=BMM_Additive
    rotationSpace=BCS_BoneSpace
    alphaBoolBlend {blendInTime=0.3, blendOption=CircularIn, blendOutTime=0.3}
    boneToModify {boneName=dyn_jacket_bk_l_02}
    rotation {pitch=-30}
   }
   Output Pose(Result=Component To Local(ComponentPose=@n21)) {layerGroup=None}
   n22 = "Transform (Modify) Bone - Bone: dyn_jacket_fr_l_MSTR"(
    ComponentPose=Local To Component(LocalPose=Copy Pose From Mesh() {bUseAttachedParent=true})
    Alpha=Get crouching alpha() -> (crouching alpha:double)
    Translation="-0.000,0.000,3"
    Rotation="52.974,-30.282,-57.791"
    Scale="1,1,1"
   ) {
    rotationMode=BMM_Additive
    rotationSpace=BCS_BoneSpace
    translationMode=BMM_Additive
    alphaBoolBlend {blendOption=Linear}
    boneToModify {boneName=dyn_jacket_fr_l_MSTR}
    rotation=<rot 52.9739 -30.2821 -57.7908>
    translation {z=3}
   }
   n23 = "Transform (Modify) Bone - Bone: dyn_jacket_fr_r_MSTR"(ComponentPose=@n22, Alpha=Get jump_alpha() -> (jump_alpha:double), Translation="0.000,0.000,4.488", Rotation="0,0,0", Scale="1,1,1") {translationMode=BMM_Additive, alphaBoolBlend {blendOption=Linear}, boneToModify {boneName=dyn_jacket_fr_r_MSTR}, translation {z=4.488}}
   n24 = "Constraint - dyn_tie_MSTR (PPPP)"!(
    ComponentPose=@n23
    Alpha=1
    ConstraintWeights_0=0.1469
    ConstraintWeights_1:float=0.0805
    ConstraintWeights_2:float=0.1429
    ConstraintWeights_3:float=0.6297
   ) -> (Pose) {
    alphaBoolBlend {blendOption=Linear}
    boneToModify {boneName=dyn_tie_MSTR}
    constraintSetup [{targetBone {boneName=clavicle_r}}, {targetBone {boneName=neck_02}}, {targetBone {boneName=neck_01}}, {targetBone {boneName=spine_05}}]
    constraintWeights [1,1,1,1]
   }
   n25 = "Transform (Modify) Bone - Bone: sleeve_R"(
    ComponentPose=@n24
    Alpha=Get Sleeve R Correction() -> (Sleeve R Correction:float)
    Translation="3.5,0,-0.35"
    Rotation="0,0,0"
    Scale="1,1,1"
   ) {
    translationMode=BMM_Additive
    translationSpace=BCS_BoneSpace
    alphaBoolBlend {blendOption=Linear}
    boneToModify {boneName=sleeve_R}
    translation {x=3.5, z=-0.35}
   }
   n26 = Control Rig(Source=Component To Local(ComponentPose=@n25)) {
    alphaBoolBlend {blendOption=Linear}
    controlRigAssetReference {blueprintRigClass=&/M_MED_TracePaper_CtrlRig.M_MED_TracePaper_CtrlRig_C}
   }
   Save cached pose 'pre_sim_cache'(Pose=@n26) {cacheName=pre_sim_cache}
  ]}
 EventGraph {nodes [
   Event BlueprintUpdateAnimation()
   > n1 = Sequence!(execute) -> (then_0, then_1, then_2:Exec, then_3:Exec, then_4:Exec)
   > n2 = Branch(Condition=Get bIsJumping())
   Set jump_alpha(execute=[@n2.then, @n3.then], jump_alpha:double=1) -> (Output_Get:double)
   n3 = Branch(execute=@n2.else, Condition=Get bIsFalling())
   Set jump_alpha(execute=@n3.else, jump_alpha=0)
   n4 = Branch(execute=@n1.then_2, Condition=Get bIsSkyDiving())
   > Set skydiving alpha(skydiving alpha:double=1) -> (Output_Get:double)
   Set skydiving alpha(execute=@n4.else, skydiving alpha=0)
   Set RightThighUpAlpha(
    execute=@n1.then_3
    RightThighUpAlpha:float=GetAlphaFromReferenceTransformDelta(
     SkeletalMeshComponent=Get BaseMesh() -> (BaseMesh:Skeletal Mesh Component Object Reference)
     BoneName=thigh_r
     BaseName=pelvis
     DeltaComponentType=EulerY
     InputRange=(X=10,Y=25)
     TwistAxis="1,0,0"
    ).RawOutput
   ) -> (Output_Get:float)
   n5 = Get BaseMesh()
   > Set Sleeve R Correction(Sleeve R Correction:float=float + float(
     A=GetAlphaFromReferenceTransformDelta(SkeletalMeshComponent=@n5, BoneName=hand_r, BaseName=lowerarm_r, DeltaComponentType=EulerY, InputRange=(X=0,Y=66), TwistAxis="1,0,0")
     B=GetAlphaFromReferenceTransformDelta(
      SkeletalMeshComponent=@n5
      BoneName=hand_r
      BaseName=lowerarm_r
      DeltaComponentType=EulerY
      InputRange=(X=0,Y=-27)
      OutputRange=(X=0,Y=0.3)
      TwistAxis="1,0,0"
     )
    )) -> (Output_Get:float)
   Set isWearingCape(execute=@n1.then_4, isWearingCape=CheckForBackbling().bIsBackblingACape)
   > Set isWearingCape_0(isWearingCape_0:bool=AND Boolean(A=Get isWearingCape() -> (isWearingCape:bool), B=Get bIsFalling())) -> (Output_Get:bool)
   n6 = Get CurrentLOD()
   n7 = Get bRigidBodyEnabled()
   Set dynamics_disabled(
    execute=@n1.then_1
    dynamics_disabled:bool=OR Boolean(A="integer > integer"(A=@n6, B=1) {is=title K2Node_CallFunction}, B=NOT Boolean(A=@n7))
   ) -> (Output_Get:bool)
   Event BlueprintInitializeAnimation()
   > Set BaseMesh(BaseMesh:Skeletal Mesh Component Object Reference=GetOwnerSkeletalMeshComponent()) -> (Output_Get:Skeletal Mesh Component Object Reference)
   Event OnEndCrouching()
   > Set crouching alpha(crouching alpha:double=0) -> (Output_Get:double)
   Event OnBeginCrouching()
   > Set crouching alpha(crouching alpha=1)
  ]}
]
variables [
 BaseMesh {type=SkeletalMeshComponent}
 crouching alpha {type=number}
 dynamics_disabled {type=boolean}
 isWearingCape {type=boolean}
 isWearingCape_0 {type=boolean}
 jump_alpha {type=number}
 RightThighUpAlpha {type=number}
 skydiving alpha {type=number}
 Sleeve R Correction {type=number}
]
