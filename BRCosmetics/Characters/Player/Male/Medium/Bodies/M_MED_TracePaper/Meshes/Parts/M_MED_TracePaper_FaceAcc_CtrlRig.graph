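# Control Rig for the TracePaper male face accessory that applies corrective offsets to the back and side dyn_hair bones.
# Spherical pose readers measure head orientation against spine_05 and weight Modify Transforms nodes across two Sequence branches.
# If nodes switch some offsets on isCrouching and isEmote, with two emote offsets at constant full weight.
# The isCrouchTargeting chain and a constant-false chain are disconnected from execution, and the isTargeting selectors are unused.
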ControlRigBlueprint /BRCosmetics/Characters/Player/Male/Medium/Bodies/M_MED_TracePaper/Meshes/Parts/M_MED_TracePaper_FaceAcc_CtrlRig
graphs [
 RigVMFunctionLibrary {links [], nodes []}
 RigVMModel {nodes [
   RigUnit_BeginExecution = &RigUnit_BeginExecution()
   RigVMFunction_Sequence = &RigVMFunction_Sequence(ExecuteContext=@RigUnit_BeginExecution.ExecutePin)
   n1 = &RigUnit_SphericalPoseReader(
    ActiveRegionScaleFactors=(PositiveWidth=0.5,NegativeWidth=0.5,PositiveHeight=0.5,NegativeHeight=0.5)
    ActiveRegionSize=0.1
    Debug=(bDrawDebug=false,bDraw2D=False,bDrawLocalAxes=False,DebugScale=25,DebugSegments=20,DebugThickness=0.25)
    DriverAxis=(X=1,Y=0,Z=0)
    DriverItem=(Type=Bone,Name="head")
    FalloffRegionScaleFactors=(PositiveWidth=1,NegativeWidth=1,PositiveHeight=1,NegativeHeight=1)
    FalloffSize=0.3
    FlipHeightScaling=False
    FlipWidthScaling=False
    OptionalParentItem=(Type=Bone,Name="spine_05")
    RotationOffset=(X=-21.3533,Y=-58.1163,Z=-15.9027)
    ExecutePin=@RigVMFunction_Sequence.A
   )
   > &RigUnit_ModifyTransforms(
    ItemToModify=((Item=(Type=Bone,Name="dyn_hair_bck_01"),Transform=(Rotation=(X=0.1074,Y=-0.0402,Z=0.2577,W=0.9594),Translation=(X=-2,Y=0,Z=0),Scale3D=(X=1,Y=1,Z=1))))
    WeightMinimum=0
    Weight=@n1.OutputParam
   )
   > &RigUnit_ModifyTransforms(
    ItemToModify=((Item=(Type=Bone,Name="dyn_hair_bck_r_01"),Transform=(Rotation=(X=0.2006,Y=-0.0479,Z=0.1258,W=0.9704),Translation=(X=-2,Y=0,Z=0),Scale3D=(X=1,Y=1,Z=1))))
    WeightMinimum=0
    Weight=@n1.OutputParam
   )
   > &RigUnit_ModifyTransforms(
    ItemToModify=((Item=(Type=Bone,Name="dyn_hair_bck_l_02"),Transform=(Rotation=(X=0,Y=0.0872,Z=0,W=0.9962),Translation=(X=-2,Y=0,Z=0),Scale3D=(X=1,Y=1,Z=1))))
    WeightMinimum=0
    Weight=@n1.OutputParam
   )
   > &RigUnit_ModifyTransforms(
    ItemToModify=((Item=(Type=Bone,Name="dyn_hair_r_side2_03"),Transform=(Rotation=(X=0.2006,Y=-0.0479,Z=0.1258,W=0.9704),Translation=(X=-2,Y=0,Z=0),Scale3D=(X=1,Y=1,Z=1))))
    WeightMinimum=0
    Weight=@n1.OutputParam
   )
   > n2 = &RigUnit_SphericalPoseReader(
    ActiveRegionScaleFactors=(PositiveWidth=0.5,NegativeWidth=0.5,PositiveHeight=1,NegativeHeight=1)
    ActiveRegionSize=0.1
    Debug=(bDrawDebug=true,bDraw2D=False,bDrawLocalAxes=False,DebugScale=25,DebugSegments=20,DebugThickness=0.25)
    DriverAxis=(X=1,Y=0,Z=0)
    DriverItem=(Type=Bone,Name="head")
    FalloffRegionScaleFactors=(PositiveWidth=1,NegativeWidth=1,PositiveHeight=1,NegativeHeight=1)
    FalloffSize=0.3
    FlipHeightScaling=False
    FlipWidthScaling=False
    OptionalParentItem=(Type=Bone,Name="spine_05")
    RotationOffset=(X=-21.3533,Y=-30.6155,Z=-15.9027)
   )
   n3 = RigVMVariableNode() {variable=isCrouching}
   DISPATCH_RigVMDispatch_If = RigVMDispatchNode(
    False=(Rotation=(X=0,Y=0,Z=0,W=1),Translation=(X=0,Y=0,Z=0),Scale3D=(X=1,Y=1,Z=1))
    True=(Rotation=(X=0.0227,Y=0.1285,Z=-0.1722,W=0.9764),Translation=(X=-5,Y=0,Z=0),Scale3D=(X=1,Y=1,Z=1))
    Condition=@n3.Value
   ) {template="DISPATCH_RigVMDispatch_If(in Condition,in True,in False,out Result)"}
   > &RigUnit_ModifyTransforms(
    ItemToModify=((Item=(Type=Bone,Name="dyn_hair_bck_01"),Transform=(Rotation=(X=0.1074,Y=-0.0402,Z=0.2577,W=0.9594),Translation=(X=-2,Y=0,Z=0),Scale3D=(X=1,Y=1,Z=1))))
    WeightMinimum=0
    ItemToModify.0.Transform.Rotation=@DISPATCH_RigVMDispatch_If.Result.Rotation
    ItemToModify.0.Transform.Translation=@DISPATCH_RigVMDispatch_If.Result.Translation
    Weight=@n2.OutputParam
   )
   n4 = RigVMDispatchNode(
    False=(Rotation=(X=0,Y=0,Z=0,W=1),Translation=(X=0,Y=0,Z=0),Scale3D=(X=1,Y=1,Z=1))
    True=(Rotation=(X=0,Y=0.1305,Z=0,W=0.9914),Translation=(X=-5,Y=0,Z=0),Scale3D=(X=1,Y=1,Z=1))
    Condition=@n3.Value
   ) {template="DISPATCH_RigVMDispatch_If(in Condition,in True,in False,out Result)"}
   > &RigUnit_ModifyTransforms(
    ItemToModify=((Item=(Type=Bone,Name="dyn_hair_bck_r_02"),Transform=(Rotation=(X=0.2006,Y=-0.0479,Z=0.1258,W=0.9704),Translation=(X=-2,Y=0,Z=0),Scale3D=(X=1,Y=1,Z=1))))
    WeightMinimum=0
    ItemToModify.0.Transform.Rotation=@n4.Result.Rotation
    ItemToModify.0.Transform.Translation=@n4.Result.Translation
    Weight=@n2.OutputParam
   )
   n5 = RigVMDispatchNode(
    False=(Rotation=(X=0,Y=0,Z=0,W=1),Translation=(X=0,Y=0,Z=0),Scale3D=(X=1,Y=1,Z=1))
    True=(Rotation=(X=0,Y=0.0436,Z=0,W=0.999),Translation=(X=-5,Y=0,Z=0),Scale3D=(X=1,Y=1,Z=1))
    Condition=@n3.Value
   ) {template="DISPATCH_RigVMDispatch_If(in Condition,in True,in False,out Result)"}
   > &RigUnit_ModifyTransforms(
    ItemToModify=((Item=(Type=Bone,Name="dyn_hair_r_side1_02"),Transform=(Rotation=(X=0.2006,Y=-0.0479,Z=0.1258,W=0.9704),Translation=(X=-2,Y=0,Z=0),Scale3D=(X=1,Y=1,Z=1))))
    WeightMinimum=0
    ItemToModify.0.Transform.Rotation=@n5.Result.Rotation
    ItemToModify.0.Transform.Translation=@n5.Result.Translation
    Weight=@n2.OutputParam
   )
   n6 = RigVMDispatchNode(
    False=(Rotation=(X=0,Y=0,Z=0,W=1),Translation=(X=0,Y=0,Z=0),Scale3D=(X=1,Y=1,Z=1))
    True=(Rotation=(X=0,Y=0.0436,Z=0,W=0.999),Translation=(X=-5,Y=0,Z=0),Scale3D=(X=1,Y=1,Z=1))
    Condition=@n3.Value
   ) {template="DISPATCH_RigVMDispatch_If(in Condition,in True,in False,out Result)"}
   > &RigUnit_ModifyTransforms(
    ItemToModify=((Item=(Type=Bone,Name="dyn_hair_bck_l_01"),Transform=(Rotation=(X=0.2006,Y=-0.0479,Z=0.1258,W=0.9704),Translation=(X=-2,Y=0,Z=0),Scale3D=(X=1,Y=1,Z=1))))
    WeightMinimum=0
    ItemToModify.0.Transform.Rotation=@n6.Result.Rotation
    ItemToModify.0.Transform.Translation=@n6.Result.Translation
    Weight=@n2.OutputParam
   )
   > n7 = &RigUnit_SphericalPoseReader(
    ActiveRegionScaleFactors=(PositiveWidth=1,NegativeWidth=1,PositiveHeight=1,NegativeHeight=1)
    ActiveRegionSize=0.1
    Debug=(bDrawDebug=false,bDraw2D=False,bDrawLocalAxes=False,DebugScale=25,DebugSegments=20,DebugThickness=0.25)
    DriverAxis=(X=1,Y=1,Z=0)
    DriverItem=(Type=Bone,Name="head")
    FalloffRegionScaleFactors=(PositiveWidth=1,NegativeWidth=1,PositiveHeight=1,NegativeHeight=1)
    FalloffSize=0.3
    FlipHeightScaling=False
    FlipWidthScaling=False
    OptionalParentItem=(Type=Bone,Name="spine_05")
    RotationOffset=(X=-90.7131,Y=0,Z=0)
   )
   n8 = RigVMVariableNode() {variable=isEmote}
   n9 = RigVMDispatchNode(
    False=(Rotation=(X=0,Y=0,Z=0,W=1),Translation=(X=0,Y=0,Z=0),Scale3D=(X=1,Y=1,Z=1))
    True=(Rotation=(X=0.0113,Y=0.0421,Z=-0.2586,W=0.965),Translation=(X=-3,Y=0,Z=1),Scale3D=(X=1,Y=1,Z=1))
    Condition=@n8.Value
   ) {template="DISPATCH_RigVMDispatch_If(in Condition,in True,in False,out Result)"}
   > &RigUnit_ModifyTransforms(
    ItemToModify=((Item=(Type=Bone,Name="dyn_hair_bck_r_01"),Transform=(Rotation=(X=0,Y=-0.5,Z=0,W=0.866),Translation=(X=0,Y=0,Z=0),Scale3D=(X=1,Y=1,Z=1))))
    WeightMinimum=0
    ItemToModify.0.Transform.Rotation=@n9.Result.Rotation
    ItemToModify.0.Transform.Translation=@n9.Result.Translation
    Weight=@n7.OutputParam
   )
   > &RigUnit_ModifyTransforms(
    ItemToModify=((Item=(Type=Bone,Name="dyn_hair_bck_r_02"),Transform=(Rotation=(X=0,Y=-0.5,Z=0,W=0.866),Translation=(X=0,Y=0,Z=0),Scale3D=(X=1,Y=1,Z=1))))
    WeightMinimum=0
    ItemToModify.0.Transform.Rotation=@n9.Result.Rotation
    ItemToModify.0.Transform.Translation=@n9.Result.Translation
    Weight=@n7.OutputParam
   )
   n10 = RigVMDispatchNode(
    False=(Rotation=(X=0,Y=0,Z=0,W=1),Translation=(X=0,Y=0,Z=0),Scale3D=(X=1,Y=1,Z=1))
    True=(Rotation=(X=0,Y=0.0436,Z=0,W=0.999),Translation=(X=-3,Y=0,Z=1),Scale3D=(X=1,Y=1,Z=1))
    Condition=@n8.Value
   ) {template="DISPATCH_RigVMDispatch_If(in Condition,in True,in False,out Result)"}
   > &RigUnit_ModifyTransforms(
    ItemToModify=((Item=(Type=Bone,Name="dyn_hair_bck_01"),Transform=(Rotation=(X=-0.0449,Y=-0.1677,Z=-0.2549,W=0.9513),Translation=(X=0,Y=0,Z=0),Scale3D=(X=1,Y=1,Z=1))))
    Weight=1
    WeightMinimum=0
    ItemToModify.0.Transform.Rotation=@n10.Result.Rotation
    ItemToModify.0.Transform.Translation=@n10.Result.Translation
   )
   > &RigUnit_ModifyTransforms(
    ItemToModify=((Item=(Type=Bone,Name="dyn_hair_bck_r_02"),Transform=(Rotation=(X=-0.0449,Y=-0.1677,Z=-0.2549,W=0.9513),Translation=(X=0,Y=0,Z=0),Scale3D=(X=1,Y=1,Z=1))))
    WeightMinimum=0
    ItemToModify.0.Transform.Rotation=@n10.Result.Rotation
    ItemToModify.0.Transform.Translation=@n10.Result.Translation
    Weight=@n7.OutputParam
   )
   n11 = RigVMDispatchNode(
    False=(Rotation=(X=0,Y=0,Z=0,W=1),Translation=(X=0,Y=0,Z=0),Scale3D=(X=1,Y=1,Z=1))
    True=(Rotation=(X=-0.0113,Y=0.0421,Z=0.2586,W=0.965),Translation=(X=-3,Y=0,Z=1),Scale3D=(X=1,Y=1,Z=1))
    Condition=@n8.Value
   ) {template="DISPATCH_RigVMDispatch_If(in Condition,in True,in False,out Result)"}
   > &RigUnit_ModifyTransforms(
    ItemToModify=((Item=(Type=Bone,Name="dyn_hair_bck_l_01"),Transform=(Rotation=(X=-0.0449,Y=-0.1677,Z=-0.2549,W=0.9513),Translation=(X=0,Y=0,Z=0),Scale3D=(X=1,Y=1,Z=1))))
    Weight=1
    WeightMinimum=0
    ItemToModify.0.Transform.Rotation=@n11.Result.Rotation
    ItemToModify.0.Transform.Translation=@n11.Result.Translation
   )
   > &RigUnit_ModifyTransforms(
    ItemToModify=((Item=(Type=Bone,Name="dyn_hair_bck_l_02"),Transform=(Rotation=(X=-0.0449,Y=-0.1677,Z=-0.2549,W=0.9513),Translation=(X=0,Y=0,Z=0),Scale3D=(X=1,Y=1,Z=1))))
    WeightMinimum=0
    ItemToModify.0.Transform.Rotation=@n11.Result.Rotation
    ItemToModify.0.Transform.Translation=@n11.Result.Translation
    Weight=@n7.OutputParam
   )
   n12 = &RigUnit_SphericalPoseReader(
    ActiveRegionScaleFactors=(PositiveWidth=1,NegativeWidth=1,PositiveHeight=1,NegativeHeight=1)
    ActiveRegionSize=0.1
    Debug=(bDrawDebug=false,bDraw2D=False,bDrawLocalAxes=False,DebugScale=25,DebugSegments=20,DebugThickness=0.25)
    DriverAxis=(X=1,Y=0,Z=0)
    DriverItem=(Type=Bone,Name="head")
    FalloffRegionScaleFactors=(PositiveWidth=1,NegativeWidth=1,PositiveHeight=1,NegativeHeight=1)
    FalloffSize=0.3
    FlipHeightScaling=False
    FlipWidthScaling=False
    OptionalParentItem=(Type=Bone,Name="spine_05")
    RotationOffset=(X=-21.3533,Y=-28.266,Z=-15.9027)
    ExecutePin=@RigVMFunction_Sequence.B
   )
   > &RigUnit_ModifyTransforms(
    ItemToModify=((Item=(Type=Bone,Name="dyn_hair_bck_02"),Transform=(Rotation=(X=-0.1964,Y=0.2936,Z=0.1524,W=0.923),Translation=(X=0,Y=0,Z=0),Scale3D=(X=1,Y=1,Z=1))))
    WeightMinimum=0
    Weight=@n12.OutputParam
   )
   > &RigUnit_ModifyTransforms(
    ItemToModify="((Item=(Type=Bone,Name=\"dyn_hair_bck_r_01\"),Transform=(Rotation=(X=0.3349,Y=0.6282,Z=-0.3613,W=0.6022),Translation=(X=0,Y=0,Z=0),Scale3D=(X=1,Y=1,Z=1))),(Item=(Type=Bone,Name=\"dyn_hair_bck_r_02\"),Transform=(Rotation=(X=0.4312,Y=0.4494,Z=-0.1494,W=0.7679),T ...(truncated)"
    WeightMinimum=0
    Weight=@n12.OutputParam
   )
   n13 = &RigUnit_SphericalPoseReader(
    ActiveRegionScaleFactors=(PositiveWidth=1,NegativeWidth=1,PositiveHeight=1,NegativeHeight=1)
    ActiveRegionSize=0.1
    Debug=(bDrawDebug=false,bDraw2D=False,bDrawLocalAxes=False,DebugScale=25,DebugSegments=20,DebugThickness=0.25)
    DriverAxis=(X=1,Y=0,Z=0)
    DriverItem=(Type=Bone,Name="head")
    ExecutePin=()
    FalloffRegionScaleFactors=(PositiveWidth=1,NegativeWidth=1,PositiveHeight=1,NegativeHeight=1)
    FalloffSize=0.3
    FlipHeightScaling=False
    FlipWidthScaling=False
    OptionalParentItem=(Type=Bone,Name="upperarm_l")
    RotationOffset=(X=43.8394,Y=-120.9734,Z=-15.9027)
   )
   n14 = RigVMVariableNode() {variable=isCrouchTargeting}
   n15 = RigVMDispatchNode(
    False=(Rotation=(X=0,Y=0,Z=0,W=1),Translation=(X=0,Y=0,Z=0),Scale3D=(X=1,Y=1,Z=1))
    True=(Rotation=(X=0.0151,Y=0.0858,Z=-0.173,W=0.9811),Translation=(X=-1,Y=0,Z=0),Scale3D=(X=1,Y=1,Z=1))
    Condition=@n14.Value
   ) {template="DISPATCH_RigVMDispatch_If(in Condition,in True,in False,out Result)"}
   > &RigUnit_ModifyTransforms(
    ItemToModify=((Item=(Type=Bone,Name="dyn_hair_bck_r_01"),Transform=(Rotation=(X=0.1074,Y=-0.0402,Z=0.2577,W=0.9594),Translation=(X=-2,Y=0,Z=0),Scale3D=(X=1,Y=1,Z=1))))
    WeightMinimum=0
    ItemToModify.0.Transform.Rotation=@n15.Result.Rotation
    ItemToModify.0.Transform.Translation=@n15.Result.Translation
    Weight=@n13.OutputParam
   )
   n16 = RigVMDispatchNode(
    False=(Rotation=(X=0,Y=0,Z=0,W=1),Translation=(X=0,Y=0,Z=0),Scale3D=(X=1,Y=1,Z=1))
    True=(Rotation=(X=0,Y=0.1305,Z=0,W=0.9914),Translation=(X=-1,Y=0,Z=0),Scale3D=(X=1,Y=1,Z=1))
    Condition=@n14.Value
   ) {template="DISPATCH_RigVMDispatch_If(in Condition,in True,in False,out Result)"}
   > &RigUnit_ModifyTransforms(
    ItemToModify=((Item=(Type=Bone,Name="dyn_hair_r_side2_01"),Transform=(Rotation=(X=0.2006,Y=-0.0479,Z=0.1258,W=0.9704),Translation=(X=-2,Y=0,Z=0),Scale3D=(X=1,Y=1,Z=1))))
    WeightMinimum=0
    ItemToModify.0.Transform.Rotation=@n16.Result.Rotation
    ItemToModify.0.Transform.Translation=@n16.Result.Translation
    Weight=@n13.OutputParam
   )
   n17 = &RigUnit_SphericalPoseReader(
    ActiveRegionScaleFactors=(PositiveWidth=1,NegativeWidth=1,PositiveHeight=1,NegativeHeight=1)
    ActiveRegionSize=0.1
    Debug=(bDrawDebug=false,bDraw2D=False,bDrawLocalAxes=False,DebugScale=25,DebugSegments=20,DebugThickness=0.25)
    DriverAxis=(X=1,Y=0,Z=0)
    DriverItem=(Type=Bone,Name="head")
    ExecutePin=()
    FalloffRegionScaleFactors=(PositiveWidth=1,NegativeWidth=1,PositiveHeight=1,NegativeHeight=1)
    FalloffSize=0.3
    FlipHeightScaling=False
    FlipWidthScaling=False
    OptionalParentItem=(Type=Bone,Name="spine_05")
    RotationOffset=(X=-37.7567,Y=3.3637,Z=-12.7493)
   )
   n18 = RigVMDispatchNode(
    Condition=False
    False=(Rotation=(X=0,Y=0,Z=0,W=1),Translation=(X=0,Y=0,Z=0),Scale3D=(X=1,Y=1,Z=1))
    True=(Rotation=(X=-0.4482,Y=0.2534,Z=-0.0998,W=0.8514),Translation=(X=-1,Y=0,Z=0),Scale3D=(X=1,Y=1,Z=1))
   ) {template="DISPATCH_RigVMDispatch_If(in Condition,in True,in False,out Result)"}
   > &RigUnit_ModifyTransforms(
    ItemToModify=((Item=(Type=Bone,Name="dyn_hair_bck_02"),Transform=(Rotation=(X=-0.0449,Y=-0.1677,Z=-0.2549,W=0.9513),Translation=(X=0,Y=0,Z=0),Scale3D=(X=1,Y=1,Z=1))))
    WeightMinimum=0
    ItemToModify.0.Transform=@n18.Result
    Weight=@n17.OutputParam
   )
   n19 = RigVMDispatchNode(
    Condition=False
    False=(Rotation=(X=0,Y=0,Z=0,W=1),Translation=(X=0,Y=0,Z=0),Scale3D=(X=1,Y=1,Z=1))
    True=(Rotation=(X=0.0588,Y=0.2261,Z=-0.2124,W=0.9488),Translation=(X=-1,Y=0,Z=0),Scale3D=(X=1,Y=1,Z=1))
   ) {template="DISPATCH_RigVMDispatch_If(in Condition,in True,in False,out Result)"}
   > &RigUnit_ModifyTransforms(
    ItemToModify=((Item=(Type=Bone,Name="dyn_hair_bck_01"),Transform=(Rotation=(X=-0.0449,Y=-0.1677,Z=-0.2549,W=0.9513),Translation=(X=0,Y=0,Z=0),Scale3D=(X=1,Y=1,Z=1))))
    WeightMinimum=0
    ItemToModify.0.Transform=@n19.Result
    Weight=@n17.OutputParam
   )
   n20 = RigVMDispatchNode(
    Condition=False
    False=(Rotation=(X=0,Y=0,Z=0,W=1),Translation=(X=0,Y=0,Z=0),Scale3D=(X=1,Y=1,Z=1))
    True=(Rotation=(X=-0.1114,Y=0.2336,Z=0.4158,W=0.8718),Translation=(X=-1,Y=0,Z=0),Scale3D=(X=1,Y=1,Z=1))
   ) {template="DISPATCH_RigVMDispatch_If(in Condition,in True,in False,out Result)"}
   > &RigUnit_ModifyTransforms(
    ItemToModify="((Item=(Type=Bone,Name=\"dyn_hair_bck_r_02\"),Transform=(Rotation=(X=-0.0449,Y=-0.1677,Z=-0.2549,W=0.9513),Translation=(X=0,Y=0,Z=0),Scale3D=(X=1,Y=1,Z=1))),(Item=(Type=Bone,Name=\"dyn_hair_bck_r_01\"),Transform=(Rotation=(X=0,Y=0,Z=0,W=1),Translation=(X=-1,Y= ...(truncated)"
    WeightMinimum=0
    ItemToModify.0.Transform=@n20.Result
    ItemToModify.1.Transform=@n20.Result
    Weight=@n17.OutputParam
   )
   n21 = RigVMDispatchNode(
    Condition=False
    False=(Rotation=(X=0,Y=0,Z=0,W=1),Translation=(X=0,Y=0,Z=0),Scale3D=(X=1,Y=1,Z=1))
    True=(Rotation=(X=0,Y=0.1736,Z=0,W=0.9848),Translation=(X=0,Y=0,Z=0),Scale3D=(X=1,Y=1,Z=1))
   ) {template="DISPATCH_RigVMDispatch_If(in Condition,in True,in False,out Result)"}
   > &RigUnit_ModifyTransforms(
    ItemToModify="((Item=(Type=Bone,Name=\"dyn_hair_l_side2_03\"),Transform=(Rotation=(X=-0.0449,Y=-0.1677,Z=-0.2549,W=0.9513),Translation=(X=0,Y=0,Z=0),Scale3D=(X=1,Y=1,Z=1))),(Item=(Type=Bone,Name=\"dyn_hair_l_side2_01\"),Transform=(Rotation=(X=0,Y=0,Z=0,W=1),Translation=(X=- ...(truncated)"
    WeightMinimum=0
    ItemToModify.0.Transform=@n21.Result
    Weight=@n17.OutputParam
   )
   n22 = RigVMVariableNode() {variable=isTargeting}
   RigVMDispatchNode(
    False=(Rotation=(X=0,Y=0,Z=0,W=1),Translation=(X=0,Y=0,Z=0),Scale3D=(X=1,Y=1,Z=1))
    True=(Rotation=(X=0.3349,Y=0.6282,Z=-0.3613,W=0.6022),Translation=(X=-2,Y=0,Z=0),Scale3D=(X=1,Y=1,Z=1))
    Condition=@n22.Value
   ) {template="DISPATCH_RigVMDispatch_If(in Condition,in True,in False,out Result)"}
   RigVMDispatchNode(
    False=(Rotation=(X=0,Y=0,Z=0,W=1),Translation=(X=0,Y=0,Z=0),Scale3D=(X=1,Y=1,Z=1))
    True=(Rotation=(X=0.4312,Y=0.4494,Z=-0.1494,W=0.7679),Translation=(X=-2,Y=0,Z=0),Scale3D=(X=1,Y=1,Z=1))
    Condition=@n22.Value
   ) {template="DISPATCH_RigVMDispatch_If(in Condition,in True,in False,out Result)"}
   RigVMDispatchNode(
    False=(Rotation=(X=0,Y=0,Z=0,W=1),Translation=(X=0,Y=0,Z=0),Scale3D=(X=1,Y=1,Z=1))
    True=(Rotation=(X=-0.1964,Y=0.2936,Z=0.1524,W=0.923),Translation=(X=-2,Y=0,Z=1),Scale3D=(X=1,Y=1,Z=1))
    Condition=@n22.Value
   ) {template="DISPATCH_RigVMDispatch_If(in Condition,in True,in False,out Result)"}
   RigVMVariableNode() {variable=isCrouching}
  ]}
]
hierarchy [
 root {loc=<0 0 0>, rot=<0 0 0 1>}
 pelvis {parent=root, loc=<0 -0.7071 93.1191>, rot=<0.0127 0.707 -0.0127 -0.707>}
 spine_01 {parent=pelvis, loc=<6.7494 -0.6388 0>, rot=<0 0 0.034 -0.9994>}
 spine_02 {parent=spine_01, loc=<7.2518 -0.035 0>, rot=<0 0 0.0238 -0.9997>}
 spine_03 {parent=spine_02, loc=<6.0104 0 0>, rot=<0 0 0.0149 -0.9999>}
 spine_04 {parent=spine_03, loc=<8.1024 0 0>, rot=<0 0 -0.0132 -0.9999>}
 spine_05 {parent=spine_04, loc=<10.0146 0 0>, rot=<0 0 -0.0015 -1>}
 neck_01 {parent=spine_05, loc=<11.3138 0.1529 0>, rot=<0 0 -0.2003 -0.9797>}
 neck_02 {parent=neck_01, loc=<4.5627 0 0>, rot=<0 0 0.0504 -0.9987>}
 head {parent=neck_02, loc=<5.3914 0 0>, rot=<0 0 0.115 -0.9934>}
 dyn_hair_bck_l_MSTR {parent=head, loc=<5.8937 -5.4336 -4.8983>, rot=<0.0265 0.3462 0.9369 -0.0413>}
 dyn_hair_bck_l_01 {parent=dyn_hair_bck_l_MSTR, loc=<0 0 0>, rot=<0 0 0 1>}
 dyn_hair_bck_l_02 {parent=dyn_hair_bck_l_01, loc=<9.2753 0 0>, rot=<0 0 0 1>}
 dyn_hair_bck_MSTR {parent=head, loc=<5.8309 -6.848 0.2699>, rot=<0.0003 0.701 0.713 -0.0153>}
 dyn_hair_bck_01 {parent=dyn_hair_bck_MSTR, loc=<0 0 0>, rot=<0 0 0 1>}
 dyn_hair_bck_02 {parent=dyn_hair_bck_01, loc=<9.2753 0 0>, rot=<0 0 0 1>}
 dyn_hair_bck_r_MSTR {parent=head, loc=<5.8356 -5.4176 4.6153>, rot=<0.0255 -0.9331 -0.3587 0.0022>}
 dyn_hair_bck_r_01 {parent=dyn_hair_bck_r_MSTR, loc=<0 0 0>, rot=<0 0 0 1>}
 dyn_hair_bck_r_02 {parent=dyn_hair_bck_r_01, loc=<9.2753 0 0>, rot=<0 0 0 1>}
 dyn_hair_r_side2_MSTR {parent=head, loc=<5.549 -1.6201 7.2442>, rot=<0.0041 -0.9972 -0.0741 0.0003>}
 dyn_hair_r_side2_01 {parent=dyn_hair_r_side2_MSTR, loc=<0 0 0>, rot=<0 0 0 1>}
 dyn_hair_r_side2_02 {parent=dyn_hair_r_side2_01, loc=<7.894 -1.198 -0.1703>, rot=<0 0.1348 0 -0.9909>}
 dyn_hair_r_side2_03 {parent=dyn_hair_r_side2_01, loc=<8.3815 1.3844 0.6162>, rot=<0 0 0 1>}
 dyn_hair_l_side2_MSTR {parent=head, loc=<6.9735 -1.5249 -7.2566>, rot=<0.0183 0.0719 0.9942 0.0779>}
 dyn_hair_l_side2_01 {parent=dyn_hair_l_side2_MSTR, loc=<0 0 0>, rot=<0 0 0 1>}
 dyn_hair_l_side2_02 {parent=dyn_hair_l_side2_01, loc=<8.0181 -0.7145 0.1359>, rot=<0 0.1348 0 -0.9909>}
 dyn_hair_l_side2_03 {parent=dyn_hair_l_side2_01, loc=<7.513 2.0245 -0.6162>, rot=<0 0 0 1>}
 dyn_hair_l_side1_MSTR {parent=head, loc=<6.6142 2.6011 -7.4396>, rot=<0 -0.0111 0.9999 -0.0041>}
 dyn_hair_l_side1_01 {parent=dyn_hair_l_side1_MSTR, loc=<0 0 0>, rot=<0 0 0 1>}
 dyn_hair_l_side1_02 {parent=dyn_hair_l_side1_01, loc=<7.5277 0 0>, rot=<0 0 0 1>}
 dyn_hair_r_side1_MSTR {parent=head, loc=<7.1612 2.6056 7.4396>, rot=<0.0041 -0.9999 0.0111 0>}
 dyn_hair_r_side1_01 {parent=dyn_hair_r_side1_MSTR, loc=<0 0 0>, rot=<0 0 0 1>}
 dyn_hair_r_side1_02 {parent=dyn_hair_r_side1_01, loc=<7.5277 0 0>, rot=<0 0 0 1>}
 dyn_hair_fr_l_MSTR {parent=head, loc=<11.1377 6.6736 -4.6218>, rot=<-0.0723 -0.1505 0.9788 -0.1188>}
 dyn_hair_fr_l_01 {parent=dyn_hair_fr_l_MSTR, loc=<0 0 0>, rot=<0 0 0 1>}
 dyn_hair_fr_l_02 {parent=dyn_hair_fr_l_01, loc=<9.2753 0 0>, rot=<0 0 0 1>}
 dyn_hair_fr_r_MSTR {parent=head, loc=<11.1351 6.9878 4.5028>, rot=<0.0924 -0.9857 0.0128 0.1404>}
 dyn_hair_fr_r_01 {parent=dyn_hair_fr_r_MSTR, loc=<0 0 0>, rot=<0 0 0 1>}
 dyn_hair_fr_r_02 {parent=dyn_hair_fr_r_01, loc=<9.2753 0 0>, rot=<0 0 0 1>}
 faceAttach {parent=head, loc=<-2.5859 0.9279 -0.0078>, rot=<0.0047 -0.7075 -0.0018 -0.7067>}
 C_brow_mid {parent=faceAttach, loc=<-0.0723 8.9797 7.1979>, rot=<-0.0005 0.0006 -0.0021 -1>}
 L_brow_mid {parent=faceAttach, loc=<2.5698 8.5892 8.0121>, rot=<-0.0007 -0.0038 0.2533 -0.9674>}
 L_brow_outer {parent=faceAttach, loc=<4.4403 6.9424 8.3039>, rot=<-0.0293 -0.0397 0.4472 -0.8931>}
 L_cheek_inner {parent=faceAttach, loc=<3.4507 7.0559 2.6249>, rot=<-0.0707 -0.2695 0.3163 -0.9068>}
 C_lip_upper_mid {parent=faceAttach, loc=<-0.0354 8.187 0.5159>, rot=<0.222 0.0023 -0.0024 -0.975>}
 L_lip_upper_outer {parent=faceAttach, loc=<1.3376 7.6523 0.5505>, rot=<0.0355 -0.0723 0.2607 -0.9621>}
 L_lip_corner {parent=faceAttach, loc=<2.0313 7.0006 0.3928>, rot=<0.0024 0.0086 -0.0167 -0.9998>}
 C_nose_bridge {parent=faceAttach, loc=<-0.0427 9.9607 4.6926>, rot=<-0.2223 0.0001 -0.0022 -0.975>}
 R_brow_mid {parent=faceAttach, loc=<-2.6752 8.5689 8.0112>, rot=<-0.0005 0.0049 -0.2573 -0.9663>}
 R_brow_outer {parent=faceAttach, loc=<-4.5443 6.8803 8.2905>, rot=<-0.0289 0.0406 -0.451 -0.8911>}
 R_cheek_inner {parent=faceAttach, loc=<-3.5209 7.0339 2.6081>, rot=<-0.0715 0.2702 -0.3202 -0.9052>}
 R_lip_corner {parent=faceAttach, loc=<-2.0851 7.1096 0.4238>, rot=<0.0025 -0.0074 0.0125 -0.9999>}
 C_jaw {parent=faceAttach, loc=<-0.0473 2.7324 1.0337>, rot=<-0.0005 0.0006 -0.0021 -1>}
 L_lip_lower_outer {parent=C_jaw, loc=<1.3903 4.6999 -0.4866>, rot=<0.013 0.0445 0.3057 -0.951>}
 C_lip_lower_mid {parent=C_jaw, loc=<0.0261 5.3862 -0.5031>, rot=<0 0 0 -1>}
 R_lip_lower_outer {parent=C_jaw, loc=<-1.3323 4.8933 -0.4916>, rot=<0.013 -0.0445 -0.3057 -0.951>}
 teeth_lower {parent=C_jaw, loc=<-0.0645 3.6527 -1.6306>, rot=<0 0 0 1>}
 tongue {parent=C_jaw, loc=<0.0318 1.435 -0.6902>, rot=<-0.1568 0 0 -0.9876>}
 R_eye {parent=faceAttach, loc=<-2.5694 6.8144 7.0021>, rot=<-0.0006 -0.0005 0.9998 -0.0195>}
 R_eye_lid_upper_mid {parent=R_eye, loc=<0 0 0>, rot=<0.0523 0 0 -0.9986>}
 R_eye_lid_lower_mid {parent=R_eye, loc=<0 0 0>, rot=<0.0329 0 0 -0.9995>}
 L_eye {parent=faceAttach, loc=<2.4811 6.8354 7.0079>, rot=<-0.0006 -0.0005 0.9999 0.0154>}
 L_eye_lid_upper_mid {parent=L_eye, loc=<0 0 0>, rot=<0.0523 0 0 -0.9986>}
 L_eye_lid_lower_mid {parent=L_eye, loc=<0 0 0>, rot=<0.0329 0 0 -0.9995>}
 teeth_upper {parent=faceAttach, loc=<-0.0396 6.6917 1.5666>, rot=<-0.0005 0.0006 -0.0021 -1>}
 R_lip_upper_outer {parent=faceAttach, loc=<-1.4444 7.6913 0.5552>, rot=<0.0355 0.0735 -0.2646 -0.9609>}
 clavicle_l {parent=spine_05, loc=<4.3575 -0.1466 -2.3223>, rot=<0.0465 -0.7289 -0.009 -0.6829>}
 upperarm_l {parent=clavicle_l, loc=<15.5111 0 0>, rot=<0.0226 -0.3234 -0.0088 -0.9459>}
 clavicle_r {parent=spine_05, loc=<4.3575 -0.1466 2.3223>, rot=<0.7289 0.0465 -0.6829 0.009>}
 upperarm_r {parent=clavicle_r, loc=<-15.5111 0 0>, rot=<0.0226 -0.3234 -0.0088 -0.9459>}
]
variables [
 isCrouchTargeting {type=bool, default=False}
 isCrouching {type=bool, default=False}
 isEmote {type=bool, default=False}
 isTargeting {type=bool, default=False}
]
