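# Physics asset for the TracePaper male medium face accessory mesh, defining capsule bodies and constraints for eight dynamic hair chains.
# Each chain runs from a MSTR bone through tightly twist-limited, swing-limited segments; both side2 chains branch, with 02 and 03 attached to 01.
# MSTR bones join the head with free angular motion, as do the head-neck and neck-spine_05 links.
# Head, neck and spine_05 capsules provide the body shapes around the hair.
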PhysicsAsset /BRCosmetics/Characters/Player/Male/Medium/Bodies/M_MED_TracePaper/Meshes/Parts/M_MED_TracePaper_FaceAcc_Physics
bodies {
 dyn_hair_bck_01 [{shapeType=Capsule, length=4.011, radius=0.989, shapeName=dyn_hair_bck_01_capsule, center=<2.1508 0 0>, rotation=<rot 90 169.975 169.975>}]
 dyn_hair_bck_02 [{
   shapeType=Capsule
   length=3.5247
   radius=2.2489
   shapeName=dyn_hair_bck_02_capsule
   center=<3.9416 0 -1.0916>
   rotation=<rot 90 153.4349 63.435>
  }]
 dyn_hair_bck_MSTR [{shapeType=Capsule, length=1, radius=1, shapeName=dyn_hair_bck_MSTR_capsule}]
 dyn_hair_bck_l_01 [{shapeType=Capsule, length=4.011, radius=0.989, shapeName=dyn_hair_bck_l_01_capsule, center=<2.1508 0 0>, rotation=<rot 90 0.0098 0.0099>}]
 dyn_hair_bck_l_02 [{
   shapeType=Capsule
   length=3.5087
   radius=2.2649
   shapeName=dyn_hair_bck_l_02_capsule
   center=<3.4505 0 -1.0916>
   rotation=<rot 90 165.9638 75.9639>
  }]
 dyn_hair_bck_l_MSTR [{shapeType=Capsule, length=1, radius=1, shapeName=dyn_hair_bck_l_MSTR_capsule}]
 dyn_hair_bck_r_01 [{
   shapeType=Capsule
   length=4.011
   radius=0.989
   shapeName=dyn_hair_bck_r_01_capsule
   center=<2.1508 0 0>
   rotation=<rot 90 -179.999 -179.9991>
  }]
 dyn_hair_bck_r_02 [{shapeType=Capsule, length=3.8, radius=1.9736, shapeName=dyn_hair_bck_r_02_capsule, center=<3.6071 0 -1.0916>, rotation=<rot 90 180 89.9998>}]
 dyn_hair_bck_r_MSTR [{shapeType=Capsule, length=1, radius=1, shapeName=dyn_hair_bck_r_MSTR_capsule}]
 dyn_hair_fr_l_01 [{
   shapeType=Capsule
   length=4.011
   radius=0.989
   shapeName=dyn_hair_fr_l_01_capsule
   center=<2.1508 0 0>
   rotation=<rot 90 -172.4477 -172.4476>
  }]
 dyn_hair_fr_l_02 [{
   shapeType=Capsule
   length=4.305
   radius=1.4686
   shapeName=dyn_hair_fr_l_02_capsule
   center=<3.9416 0 -1.0916>
   rotation=<rot 90 -179.9998 -179.9991>
  }]
 dyn_hair_fr_l_MSTR [{shapeType=Capsule, length=1, radius=1, shapeName=dyn_hair_fr_l_MSTR_capsule}]
 dyn_hair_l_side1_01 [{shapeType=Capsule, length=4.011, radius=0.989, shapeName=dyn_hair_l_side1_01_capsule, center=<2.1508 0 0>, rotation=<rot 90 0 0.0001>}]
 dyn_hair_l_side1_02 [{
   shapeType=Capsule
   length=4.305
   radius=1.4686
   shapeName=dyn_hair_l_side1_02_capsule
   center=<3.123 0 -1.0916>
   rotation=<rot 90 -165.9638 104.0373>
  }]
 dyn_hair_l_side1_MSTR [{shapeType=Capsule, length=1, radius=1, shapeName=dyn_hair_l_side1_MSTR_capsule}]
 dyn_hair_l_side2_01 [{
   shapeType=Capsule
   length=4.011
   radius=0.989
   shapeName=dyn_hair_l_side2_01_capsule
   center=<2.1508 0 0>
   rotation=<rot 90 179.9802 179.9803>
  }]
 dyn_hair_l_side2_02 [{
   shapeType=Capsule
   length=4.305
   radius=1.4686
   shapeName=dyn_hair_l_side2_02_capsule
   center=<4.5223 0 -1.0916>
   rotation=<rot 90 165.9638 75.964>
  }]
 dyn_hair_l_side2_03 [{
   shapeType=Capsule
   length=4.305
   radius=1.4686
   shapeName=dyn_hair_l_side2_03_capsule
   center=<4.5762 0 -1.0916>
   rotation=<rot 90 -165.9638 104.0365>
  }]
 dyn_hair_l_side2_MSTR [{shapeType=Capsule, length=1, radius=1, shapeName=dyn_hair_l_side2_MSTR_capsule}]
 dyn_hair_r_side1_01 [{shapeType=Capsule, length=4.011, radius=0.989, shapeName=dyn_hair_r_side1_01_capsule, center=<2.1508 0 0>, rotation=<rot 90 0 -0.0001>}]
 dyn_hair_r_side1_02 [{
   shapeType=Capsule
   length=4.305
   radius=1.4686
   shapeName=dyn_hair_r_side1_02_capsule
   center=<3.9416 0 -1.0916>
   rotation=<rot 90 0 89.9989>
  }]
 dyn_hair_r_side1_MSTR [{shapeType=Capsule, length=1, radius=1, shapeName=dyn_hair_r_side1_MSTR_capsule}]
 dyn_hair_r_side2_01 [{shapeType=Capsule, length=4.011, radius=0.989, shapeName=dyn_hair_r_side2_01_capsule, center=<2.1508 0 0>, rotation=<rot 90 0 -0.0001>}]
 dyn_hair_r_side2_02 [{
   shapeType=Capsule
   length=3.6156
   radius=2.158
   shapeName=dyn_hair_r_side2_02_capsule
   center=<3.7128 0 -1.0916>
   rotation=<rot 90 172.875 82.8747>
  }]
 dyn_hair_r_side2_03 [{
   shapeType=Capsule
   length=3.8285
   radius=1.9451
   shapeName=dyn_hair_r_side2_03_capsule
   center=<3.6314 0 -1.0916>
   rotation=<rot 90 175.2364 85.2362>
  }]
 dyn_hair_r_side2_MSTR [{shapeType=Capsule, length=1, radius=1, shapeName=dyn_hair_r_side2_MSTR_capsule}]
 head [{shapeType=Capsule, length=10.2898, radius=5.8664, shapeName=head_capsule, center=<0.4314 2.5381 0.0002>, rotation=<rot 90 180 174>}]
 neck_01 [{shapeType=Capsule, length=8.5151, radius=7.8059, shapeName=neck_01_capsule, center=<-2.9733 1.4382 0>, rotation=<rot 90 0 -10.5276>}]
 spine_05 [{
   shapeType=Capsule
   length=29.0187
   radius=8.3338
   shapeName=spine_05_capsule
   center=<2.3335 -3.2776 -0.0023>
   rotation=<rot 0.0218 34.999 0>
  }]
}
constraints [
 {bone1Name=dyn_hair_bck_01, bone2Name=dyn_hair_bck_MSTR, swing1LimitDegrees=10, swing2LimitDegrees=45, twistLimitDegrees=1}
 {bone1Name=dyn_hair_bck_02, bone2Name=dyn_hair_bck_01, swing1LimitDegrees=10, swing2LimitDegrees=45, twistLimitDegrees=1}
 {bone1Name=dyn_hair_bck_l_01, bone2Name=dyn_hair_bck_l_MSTR, swing1LimitDegrees=10, swing2LimitDegrees=45, twistLimitDegrees=1}
 {bone1Name=dyn_hair_bck_l_02, bone2Name=dyn_hair_bck_l_01, swing1LimitDegrees=10, swing2LimitDegrees=45, twistLimitDegrees=1}
 {bone1Name=dyn_hair_bck_r_01, bone2Name=dyn_hair_bck_r_MSTR, swing1LimitDegrees=10, swing2LimitDegrees=45, twistLimitDegrees=1}
 {bone1Name=dyn_hair_bck_r_02, bone2Name=dyn_hair_bck_r_01, swing1LimitDegrees=10, swing2LimitDegrees=45, twistLimitDegrees=1}
 {bone1Name=dyn_hair_fr_l_01, bone2Name=dyn_hair_fr_l_MSTR, swing1LimitDegrees=20, swing2LimitDegrees=10, twistLimitDegrees=1}
 {bone1Name=dyn_hair_fr_l_02, bone2Name=dyn_hair_fr_l_01, swing1LimitDegrees=20, swing2LimitDegrees=10, twistLimitDegrees=1}
 {bone1Name=dyn_hair_l_side1_01, bone2Name=dyn_hair_l_side1_MSTR, swing1LimitDegrees=10, swing2LimitDegrees=45, twistLimitDegrees=1}
 {bone1Name=dyn_hair_l_side1_02, bone2Name=dyn_hair_l_side1_01, swing1LimitDegrees=10, swing2LimitDegrees=45, twistLimitDegrees=1}
 {bone1Name=dyn_hair_l_side2_01, bone2Name=dyn_hair_l_side2_MSTR, swing1LimitDegrees=20, swing2LimitDegrees=45, twistLimitDegrees=1}
 {bone1Name=dyn_hair_l_side2_02, bone2Name=dyn_hair_l_side2_01, swing1LimitDegrees=10, swing2LimitDegrees=45, twistLimitDegrees=1}
 {bone1Name=dyn_hair_l_side2_03, bone2Name=dyn_hair_l_side2_01, swing1LimitDegrees=10, swing2LimitDegrees=45, twistLimitDegrees=1}
 {bone1Name=dyn_hair_r_side1_01, bone2Name=dyn_hair_r_side1_MSTR, swing1LimitDegrees=10, swing2LimitDegrees=45, twistLimitDegrees=1}
 {bone1Name=dyn_hair_r_side1_02, bone2Name=dyn_hair_r_side1_01, swing1LimitDegrees=10, swing2LimitDegrees=45, twistLimitDegrees=1}
 {bone1Name=dyn_hair_r_side2_01, bone2Name=dyn_hair_r_side2_MSTR, swing1LimitDegrees=20, swing2LimitDegrees=45, twistLimitDegrees=1}
 {bone1Name=dyn_hair_r_side2_02, bone2Name=dyn_hair_r_side2_01, swing1LimitDegrees=10, swing2LimitDegrees=45, twistLimitDegrees=1}
 {bone1Name=dyn_hair_r_side2_03, bone2Name=dyn_hair_r_side2_01, swing1LimitDegrees=10, swing2LimitDegrees=45, twistLimitDegrees=1}
 {
  bone1Name=dyn_hair_bck_MSTR
  bone2Name=head
  swing1LimitDegrees=45
  swing1Motion=Free
  swing2LimitDegrees=45
  swing2Motion=Free
  twistLimitDegrees=45
  twistMotion=Free
 }
 {
  bone1Name=dyn_hair_bck_l_MSTR
  bone2Name=head
  swing1LimitDegrees=45
  swing1Motion=Free
  swing2LimitDegrees=45
  swing2Motion=Free
  twistLimitDegrees=45
  twistMotion=Free
 }
 {
  bone1Name=dyn_hair_bck_r_MSTR
  bone2Name=head
  swing1LimitDegrees=45
  swing1Motion=Free
  swing2LimitDegrees=45
  swing2Motion=Free
  twistLimitDegrees=45
  twistMotion=Free
 }
 {
  bone1Name=dyn_hair_fr_l_MSTR
  bone2Name=head
  swing1LimitDegrees=45
  swing1Motion=Free
  swing2LimitDegrees=45
  swing2Motion=Free
  twistLimitDegrees=45
  twistMotion=Free
 }
 {
  bone1Name=dyn_hair_l_side1_MSTR
  bone2Name=head
  swing1LimitDegrees=45
  swing1Motion=Free
  swing2LimitDegrees=45
  swing2Motion=Free
  twistLimitDegrees=45
  twistMotion=Free
 }
 {
  bone1Name=dyn_hair_l_side2_MSTR
  bone2Name=head
  swing1LimitDegrees=45
  swing1Motion=Free
  swing2LimitDegrees=45
  swing2Motion=Free
  twistLimitDegrees=45
  twistMotion=Free
 }
 {
  bone1Name=dyn_hair_r_side1_MSTR
  bone2Name=head
  swing1LimitDegrees=45
  swing1Motion=Free
  swing2LimitDegrees=45
  swing2Motion=Free
  twistLimitDegrees=45
  twistMotion=Free
 }
 {
  bone1Name=dyn_hair_r_side2_MSTR
  bone2Name=head
  swing1LimitDegrees=45
  swing1Motion=Free
  swing2LimitDegrees=45
  swing2Motion=Free
  twistLimitDegrees=45
  twistMotion=Free
 }
 {
  bone1Name=head
  bone2Name=neck_01
  swing1LimitDegrees=45
  swing1Motion=Free
  swing2LimitDegrees=45
  swing2Motion=Free
  twistLimitDegrees=45
  twistMotion=Free
 }
 {
  bone1Name=neck_01
  bone2Name=spine_05
  swing1LimitDegrees=45
  swing1Motion=Free
  swing2LimitDegrees=45
  swing2Motion=Free
  twistLimitDegrees=45
  twistMotion=Free
 }
]
