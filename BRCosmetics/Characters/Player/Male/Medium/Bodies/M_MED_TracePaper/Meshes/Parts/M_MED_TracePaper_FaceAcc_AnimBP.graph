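# Animation Blueprint for the TracePaper male character's face accessory part, layering secondary motion onto the pose copied from the attached parent mesh.
# The pose runs through a spine-relative RigidBody simulation, a Control Rig driven by targeting, crouch and emote states, then a spring AnimDynamics front-right hair chain with head and neck sphere limits.
# Each update sets CrouchTargeting from targeting plus crouch-sprinting, and computes the gravity override for both simulations through a deprecated helper.
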AnimBlueprint /BRCosmetics/Characters/Player/Male/Medium/Bodies/M_MED_TracePaper/Meshes/Parts/M_MED_TracePaper_FaceAcc_AnimBP
target_skeleton=&/M_MED_TracePaper_FaceAcc_Skeleton
graphs [
 AnimGraph {nodes [
   n1 = Get Gravity Override()
   n2 = RigidBody!(
    ComponentPose=Local To Component(LocalPose=Copy Pose From Mesh() {bUseAttachedParent=true})
    Alpha
    OverrideWorldGravity:Vector=@n1
    ExternalForce
   ) -> (Pose) {
    bClampLinearTranslationLimitToRefPose=true
    bOverrideWorldGravity=true
    overridePhysicsAsset=&/M_MED_TracePaper_FaceAcc_Physics
    simulationSpace=BaseBoneSpace
    alphaBoolBlend {blendOption=Linear}
    baseBoneRef {boneName=spine_03}
    componentAppliedLinearAccClamp=<2000 2000 2000>
    componentLinearAccScale=<0.3 0.3 0.3>
    componentLinearVelScale=<0.3 0.3 0.3>
    overrideWorldGravity {z=-980}
   }
   n3 = Control Rig!(
    Source=Component To Local(ComponentPose=@n2)
    Alpha
    isTargeting:bool=Get bIsTargeting()
    isCrouching:bool=Get bIsCrouching()
    isCrouchTargeting:bool=Get CrouchTargeting() -> (CrouchTargeting:bool)
    isEmote:bool=Get bIsPlayingEmote()
   ) -> (Pose) {
    alphaBoolBlend {blendOption=Linear}
    controlRigAssetReference {blueprintRigClass=&/M_MED_TracePaper_FaceAcc_CtrlRig.M_MED_TracePaper_FaceAcc_CtrlRig_C}
   }
   n4 = "Anim Dynamics - Chain: dyn_hair_fr_r_01 -> dyn_hair_fr_r_02"!(ComponentPose=Local To Component(LocalPose=@n3), Alpha=1, GravityOverride:Vector=@n1, ExternalForce="0,0,0") -> (Pose) {
    angularDampingOverride=0.9
    angularSpringConstant=350
    bAngularSpring=true
    bChain=true
    bOverrideAngularDamping=true
    bUseGravityOverride=true
    bUseSphericalLimits=true
    simulationSpace=BoneRelative
    alphaBoolBlend {blendOption=Linear}
    boundBone {boneName=dyn_hair_fr_r_01}
    chainEnd {boneName=dyn_hair_fr_r_02}
    componentAppliedLinearAccClamp=<2000 2000 2000>
    componentLinearAccScale=<0.3 0.3 0.3>
    componentLinearVelScale=<0.3 0.3 0.3>
    physicsBodyDefinitions [
     {
      boundBone {boneName=dyn_hair_fr_r_01}
      boxExtents=<30 30 30>
      constraintSetup {angularLimitsMax=<0 0 20>, angularLimitsMin=<0 0 -10>, angularTarget=<0 0 0>}
      localJointOffset=<0.5 0 0>
     }
     {
      boundBone {boneName=dyn_hair_fr_r_02}
      boxExtents=<30 30 30>
      constraintSetup {angularLimitsMax=<0 0 30>, angularLimitsMin=<0 0 -15>, angularTarget=<0 0 0>}
      localJointOffset=<0.5 0 0>
     }
    ]
    sphericalLimits [
     {limitRadius=8.4993, drivingBone {boneName=head}, sphereLocalOffset=<4.0179 2.4084 -0.749>}
     {limitRadius=8.3244, drivingBone {boneName=neck_02}, sphereLocalOffset=<0.4101 1.9647 -0.983>}
    ]
   }
   Output Pose(Result=Component To Local(ComponentPose=@n4)) {layerGroup=None}
  ]}
 EventGraph {nodes [
   Event BlueprintUpdateAnimation()
   > Set CrouchTargeting(CrouchTargeting:bool=AND Boolean(A=Get bIsTargeting(), B=Get bIsCrouchSprinting())) -> (Output_Get:bool)
   > n1 = Gravity Override State Function DEPRECATED(
    CharacterPartAnimInstance=Self-Reference()
    GameTimeInSeconds=Get Game Time in Seconds() -> (Game Time in Seconds:double)
    GravityOverrideParams=Get Gravity Override Params()
   )
   > Set Gravity Override(Gravity Override=@n1)
  ]}
]
variables [
 CrouchTargeting {type=boolean}
 Game Time in Seconds {type=number}
 Gravity Override {type=Vector}
 Gravity Override Params {type=GravityOverrideParamsStruct}
]
